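# Physics asset for the CrystalGlobe pickaxe melee weapon mesh, a cosmetic axe.
# It has a single body on the b_AR_Root bone, made of one capsule rotated and offset from the bone origin.
# There are no constraints, so the mesh's collision is represented by that one capsule alone.
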PhysicsAsset /BRCosmetics/Weapons/FORT_Melee/Pickaxe_CrystalGlobe/Meshes/CrystalGlobe_Axe_PhysicsAsset
bodies {b_AR_Root [{shapeType=Capsule, length=34.8047, radius=9.0209, center=<-13.0352 0.2587 3.0596>, rotation=<rot 0 -90.0317 -87.198>}]}
constraints []
